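# Looping daytime outdoor ambience cue for the Coney Crossroads platform top, reusing NoSweatAerialHQ platform beds and a Hollow_Base wind-rattled metal loop.
# A quality node picks quad bed plus metal wind at the first level, quad alone at the second, and stereo bed for the last two.
# Nested looping envelopers of differing lengths slowly drift pitch and volume, swelling the metal layer irregularly to disguise repetition.
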SoundCue /Game/Sounds/Artemis/POI/Chrome_ConeyCrossroads/Ambience_ConeyCrossRoads_PlatformTop_Day_Loop_Cue
root=QualityLevel_0
nodes [
 {
  id=Enveloper_0
  children [WavePlayer_0]
  properties {
   bLoop=true
   bLoopIndefinitely=true
   loopEnd=30
   pitchCurve {editorCurveData {keys [{interpMode=RCIM_Cubic, time=0}, <at 15.9468 0.9181 RCIM_Cubic>, {interpMode=RCIM_Cubic, time=30}]}}
   volumeCurve {editorCurveData {keys [<at 0 0>,<at 2 1>,<at 15 1>,<at 28 1>,<at 30 0>]}}
  }
 }
 {
  id=Enveloper_3
  children [Enveloper_0]
  properties {
   bLoop=true
   bLoopIndefinitely=true
   loopEnd=60
   pitchCurve {editorCurveData {keys [{time=0}]}}
   volumeCurve {editorCurveData {keys [
      {arriveTangent=-0.01, interpMode=RCIM_Cubic, time=0, value=1}
      {arriveTangent=-0.01, interpMode=RCIM_Cubic, leaveTangent=-0.01, time=10.1099, value=0.521}
      {arriveTangent=0.0115, interpMode=RCIM_Cubic, leaveTangent=0.0115, tangentMode=RCTM_User, time=20, value=0.8}
      {arriveTangent=0.0133, interpMode=RCIM_Cubic, leaveTangent=0.0133, time=30, value=0.1}
      <at 35 1 RCIM_Cubic>
      {arriveTangent=0.0054, interpMode=RCIM_Cubic, leaveTangent=0.0054, tangentMode=RCTM_User, time=40, value=0.1}
      {arriveTangent=-0.0048, interpMode=RCIM_Cubic, leaveTangent=-0.0048, tangentMode=RCTM_User, time=43.6255, value=0.588}
      {arriveTangent=0.0011, interpMode=RCIM_Cubic, leaveTangent=0.0011, tangentMode=RCTM_User, time=50.8497, value=0.4875}
      <at 60 1 RCIM_Cubic>
     ]}}
  }
 }
 {
  id=Enveloper_4
  children [WavePlayer_1]
  properties {
   bLoop=true
   bLoopIndefinitely=true
   loopEnd=40
   pitchCurve {editorCurveData {keys [
      {arriveTangent=0, interpMode=RCIM_Cubic, leaveTangent=0, tangentMode=RCTM_User, time=0}
      {arriveTangent=0, interpMode=RCIM_Cubic, leaveTangent=0, tangentMode=RCTM_User, time=20, value=0.92}
      {arriveTangent=0.004, interpMode=RCIM_Cubic, leaveTangent=0.004, time=40}
     ]}}
   volumeCurve {editorCurveData {keys [<at 0 1>]}}
  }
 }
 {
  id=Enveloper_5
  children [WavePlayer_3]
  properties {
   bLoop=true
   bLoopIndefinitely=true
   loopEnd=40
   pitchCurve {editorCurveData {keys [
      {arriveTangent=0, interpMode=RCIM_Cubic, leaveTangent=0, tangentMode=RCTM_User, time=0}
      {arriveTangent=0, interpMode=RCIM_Cubic, leaveTangent=0, tangentMode=RCTM_User, time=20, value=0.92}
      {arriveTangent=0.004, interpMode=RCIM_Cubic, leaveTangent=0.004, time=40}
     ]}}
   volumeCurve {editorCurveData {keys [<at 0 1>]}}
  }
 }
 {
  id=Enveloper_6
  children [Enveloper_5]
  properties {
   bLoop=true
   bLoopIndefinitely=true
   loopEnd=30
   pitchCurve {editorCurveData {keys [{time=0}]}}
   volumeCurve {editorCurveData {keys [
      <at 0 1 RCIM_Cubic>
      {arriveTangent=0, interpMode=RCIM_Cubic, leaveTangent=0, tangentMode=RCTM_User, time=15, value=0.9}
      <at 30 1 RCIM_Cubic>
     ]}}
  }
 }
 {
  id=Enveloper_7
  children [Enveloper_4]
  properties {
   bLoop=true
   bLoopIndefinitely=true
   loopEnd=30
   pitchCurve {editorCurveData {keys [{time=0}]}}
   volumeCurve {editorCurveData {keys [
      <at 0 1 RCIM_Cubic>
      {arriveTangent=0, interpMode=RCIM_Cubic, leaveTangent=0, tangentMode=RCTM_User, time=15, value=0.9}
      <at 30 1 RCIM_Cubic>
     ]}}
  }
 }
 {id=Mixer_0, children [Enveloper_3,Enveloper_7], properties {inputVolume [1,1]}}
 {id=QualityLevel_0, children [Mixer_0,Enveloper_7,Enveloper_6,Enveloper_6]}
 {
  id=WavePlayer_0
  properties {bLooping=true, soundWaveAssetPtr=/Game/Sounds/Artemis/POI/Hollow_Base/Spots/Spot_Wind_Excited_Metal_Vibrating_Nylon_Loop}
 }
 {
  id=WavePlayer_1
  properties {bLooping=true, soundWaveAssetPtr=/Game/Sounds/Artemis/POI/NoSweatAerialHQ/Ambience_NoSweatAerialHQ_PlatformTop_Day_Quad}
 }
 {
  id=WavePlayer_3
  properties {bLooping=true, soundWaveAssetPtr=/Game/Sounds/Artemis/POI/NoSweatAerialHQ/Ambience_NoSweatAerialHQ_PlatformTop_Day_Stereo}
 }
]
properties {
 attenuationSettings=&Fort_Ambient_Athena_Att
 maxDistance=2097152
 soundClassObject=&Fort_Ambient_Outdoors_Athena
 volumeMultiplier=0.501
 preEffectBusSends [{}]
}
