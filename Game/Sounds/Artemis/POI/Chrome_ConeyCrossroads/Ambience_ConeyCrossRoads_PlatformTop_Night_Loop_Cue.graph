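# Looping outdoor night ambience cue for the Coney Crossroads POI's platform top, built from waves borrowed from NoSweatAerialHQ and AirborneCity.
# A quality switch selects a quad bed mixed with a wind-excited metal loop, the quad bed alone, or a stereo bed for the last two levels.
# Nested, indefinitely looping envelopers with differing periods modulate pitch and volume, giving the metal layer irregular swells and the beds slow drift.
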SoundCue /Game/Sounds/Artemis/POI/Chrome_ConeyCrossroads/Ambience_ConeyCrossRoads_PlatformTop_Night_Loop_Cue
root=QualityLevel_0
nodes [
 {
  id=Enveloper_0
  children [WavePlayer_23]
  properties {
   bLoop=true
   bLoopIndefinitely=true
   loopEnd=30
   pitchCurve {editorCurveData {keys [{interpMode=RCIM_Cubic, time=0}, <at 15.9468 0.9181 RCIM_Cubic>, {interpMode=RCIM_Cubic, time=30}]}}
   volumeCurve {editorCurveData {keys [<at 0 1>,<at 15 1.4>,<at 30 1>]}}
  }
 }
 {
  id=Enveloper_3
  children [Enveloper_0]
  properties {
   bLoop=true
   bLoopIndefinitely=true
   loopEnd=60
   pitchCurve {editorCurveData {keys [{time=0}]}}
   volumeCurve {editorCurveData {keys [
      {arriveTangent=-0.01, interpMode=RCIM_Cubic, time=0, value=1}
      {arriveTangent=-0.0537, interpMode=RCIM_Cubic, leaveTangent=-0.0537, time=10, value=1.2}
      {arriveTangent=0.001, interpMode=RCIM_Cubic, leaveTangent=0.001, tangentMode=RCTM_User, time=14.8985, value=0.2}
      {arriveTangent=-0.0072, interpMode=RCIM_Cubic, leaveTangent=-0.0072, time=20, value=1}
      {arriveTangent=-0.0286, interpMode=RCIM_Cubic, leaveTangent=-0.0286, tangentMode=RCTM_User, time=25.6275, value=0.123}
      {arriveTangent=0.1042, interpMode=RCIM_Cubic, leaveTangent=0.1042, time=30.2698, value=0.1}
      {arriveTangent=0.0271, interpMode=RCIM_Cubic, leaveTangent=0.0271, time=35, value=1.1}
      <at 40 0.364>
      {arriveTangent=0.0318, interpMode=RCIM_Cubic, leaveTangent=0.0318, time=52.0586, value=0.3704}
      <at 60 1>
     ]}}
  }
 }
 {
  id=Enveloper_4
  children [WavePlayer_3]
  properties {
   bLoop=true
   bLoopIndefinitely=true
   loopEnd=40
   pitchCurve {editorCurveData {keys [
      {arriveTangent=0, interpMode=RCIM_Cubic, leaveTangent=0, tangentMode=RCTM_User, time=0}
      {arriveTangent=0, interpMode=RCIM_Cubic, leaveTangent=0, tangentMode=RCTM_User, time=20, value=0.92}
      {arriveTangent=0.004, interpMode=RCIM_Cubic, leaveTangent=0.004, time=40}
     ]}}
   volumeCurve {editorCurveData {keys [<at 0 1>]}}
  }
 }
 {
  id=Enveloper_5
  children [WavePlayer_1]
  properties {
   bLoop=true
   bLoopIndefinitely=true
   loopEnd=40
   pitchCurve {editorCurveData {keys [
      {arriveTangent=0, interpMode=RCIM_Cubic, leaveTangent=0, tangentMode=RCTM_User, time=0}
      {arriveTangent=0, interpMode=RCIM_Cubic, leaveTangent=0, tangentMode=RCTM_User, time=20, value=0.92}
      {arriveTangent=0.004, interpMode=RCIM_Cubic, leaveTangent=0.004, time=40}
     ]}}
   volumeCurve {editorCurveData {keys [<at 0 1>]}}
  }
 }
 {
  id=Enveloper_6
  children [Enveloper_5]
  properties {
   bLoop=true
   bLoopIndefinitely=true
   loopEnd=30
   pitchCurve {editorCurveData {keys [{time=0}]}}
   volumeCurve {editorCurveData {keys [
      <at 0 1 RCIM_Cubic>
      {arriveTangent=0, interpMode=RCIM_Cubic, leaveTangent=0, tangentMode=RCTM_User, time=15, value=0.9}
      <at 30 1 RCIM_Cubic>
     ]}}
  }
 }
 {
  id=Enveloper_7
  children [Enveloper_4]
  properties {
   bLoop=true
   bLoopIndefinitely=true
   loopEnd=30
   pitchCurve {editorCurveData {keys [{time=0}]}}
   volumeCurve {editorCurveData {keys [
      <at 0 1 RCIM_Cubic>
      {arriveTangent=0, interpMode=RCIM_Cubic, leaveTangent=0, tangentMode=RCTM_User, time=15, value=0.9}
      <at 30 1 RCIM_Cubic>
     ]}}
  }
 }
 {id=Mixer_0, children [Enveloper_3,Enveloper_7], properties {inputVolume [0.6,1.1]}}
 {id=QualityLevel_0, children [Mixer_0,Enveloper_7,Enveloper_6,Enveloper_6]}
 {
  id=WavePlayer_1
  properties {bLooping=true, soundWaveAssetPtr=/Game/Sounds/Artemis/POI/NoSweatAerialHQ/Ambience_NoSweatAerialHQ_PlatformTop_Night_Stereo}
 }
 {
  id=WavePlayer_23
  properties {bLooping=true, soundWaveAssetPtr=/Game/Sounds/Artemis/POI/AirborneCity/Spots/Spot_AirborneCity_Wind_Excited_Metal_Loop_02}
 }
 {
  id=WavePlayer_3
  properties {bLooping=true, soundWaveAssetPtr=/Game/Sounds/Artemis/POI/NoSweatAerialHQ/Ambience_NoSweatAerialHQ_PlatformTop_Night_Quad}
 }
]
properties {
 attenuationSettings=&Fort_Ambient_Athena_Att
 maxDistance=2097152
 soundClassObject=&Fort_Ambient_Outdoors_Athena
 volumeMultiplier=0.451
 preEffectBusSends [{}]
}
