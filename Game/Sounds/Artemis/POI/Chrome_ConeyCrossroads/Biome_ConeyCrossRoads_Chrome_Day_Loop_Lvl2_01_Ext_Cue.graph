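# Looping daytime exterior ambience cue for the Coney Crossroads Chrome POI, routed through the Athena ambient sound class and attenuation.
# A quality-level node plays the quad wave for the first two levels and the stereo wave for the last two.
# Each branch stacks two endlessly looping envelopers, one gently drifting pitch and one dipping volume, to vary the loop.
# Despite the Lvl2 name, both waves are Chrome biome Lvl1 loops.
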SoundCue /Game/Sounds/Artemis/POI/Chrome_ConeyCrossroads/Biome_ConeyCrossRoads_Chrome_Day_Loop_Lvl2_01_Ext_Cue
root=QualityLevel_0
nodes [
 {
  id=Enveloper_1
  children [WavePlayer_1]
  properties {
   bLoop=true
   bLoopIndefinitely=true
   loopEnd=60
   pitchCurve {editorCurveData {keys [
      {interpMode=RCIM_Cubic, time=0}
      {arriveTangent=0.0002, interpMode=RCIM_Cubic, leaveTangent=0.0002, tangentMode=RCTM_User, time=25, value=0.98}
      {arriveTangent=0.0006, interpMode=RCIM_Cubic, leaveTangent=0.0006, tangentMode=RCTM_User, time=42.7605, value=1.02}
      {interpMode=RCIM_Cubic, time=60}
     ]}}
   volumeCurve {editorCurveData {keys [<at 0 1>]}}
  }
 }
 {
  id=Enveloper_2
  children [Enveloper_1]
  properties {
   bLoop=true
   bLoopIndefinitely=true
   loopEnd=20
   pitchCurve {editorCurveData {keys [{time=0}]}}
   volumeCurve {editorCurveData {keys [
      {arriveTangent=-0.01, interpMode=RCIM_Cubic, time=0, value=1}
      <at 10 0.9 RCIM_Cubic>
      {interpMode=RCIM_Cubic, leaveTangent=0.01, time=20, value=1}
     ]}}
  }
 }
 {
  id=Enveloper_3
  children [Enveloper_4]
  properties {
   bLoop=true
   bLoopIndefinitely=true
   loopEnd=30
   pitchCurve {editorCurveData {keys [{time=0}]}}
   volumeCurve {editorCurveData {keys [
      {arriveTangent=-0.01, interpMode=RCIM_Cubic, leaveTangent=-0.01, time=0, value=1}
      <at 10 0.9 RCIM_Cubic>
      {arriveTangent=0.01, interpMode=RCIM_Cubic, leaveTangent=0.01, time=20, value=1}
     ]}}
  }
 }
 {
  id=Enveloper_4
  children [WavePlayer_0]
  properties {
   bLoop=true
   bLoopIndefinitely=true
   loopEnd=60
   pitchCurve {editorCurveData {keys [
      {interpMode=RCIM_Cubic, time=0}
      {arriveTangent=0.0002, interpMode=RCIM_Cubic, leaveTangent=0.0002, tangentMode=RCTM_User, time=30, value=0.95}
      {arriveTangent=0.0006, interpMode=RCIM_Cubic, leaveTangent=0.0006, tangentMode=RCTM_User, time=42.7605, value=1.05}
      {interpMode=RCIM_Cubic, time=60}
     ]}}
   volumeCurve {editorCurveData {keys [<at 0 1>]}}
  }
 }
 {id=QualityLevel_0, children [Enveloper_2,Enveloper_2,Enveloper_3,Enveloper_3]}
 {id=WavePlayer_0, properties {bLooping=true, soundWaveAssetPtr=/Game/Sounds/Artemis/Biomes/Chrome/AMB_Chrome_Day_Loop_Lvl1_04_Stereo}}
 {id=WavePlayer_1, properties {bLooping=true, soundWaveAssetPtr=/Game/Sounds/Artemis/Biomes/Chrome/AMB_Chrome_Day_Loop_Lvl1_04_Quad}}
]
properties {
 attenuationSettings=&Fort_Ambient_Athena_Att
 maxDistance=2097152
 soundClassObject=&Fort_Ambient_Athena
 volumeMultiplier=0.852
 preEffectBusSends [{}]
}
